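# Animation Blueprint for the Binary Twin backpack cosmetic, posing its wings to match the player's movement.
# The wingPosition state machine selects pose-asset poses for base, targeting, jump, fall, crouch and a nested skydiving machine (parachute, dive, float).
# Most states add acceleration-driven lean blendspaces.
# The cached result gets interpolated Run and Sprint additives from speed remaps, bypassed while skydiving or crouching.
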AnimBlueprint /BRCosmetics/Accessories/FORT_Backpacks/Backpack_F_MED_Binary_Twin/Meshes/F_MED_Binary_Twin_Pack_AnimBP
target_skeleton=/Game/Accessories/FORT_Backpacks/Backpack_F_MED_Binary_Twin/Meshes/F_MED_Binary_Twin_Pack_Skeleton
graphs [
 AnimGraph {nodes [
   n1 = Binary_Twin_Movement_Poses_PoseAsset_Add : Run!(PoseWeight:float=1) -> (Pose:Pose Link Structure) {
    is=title AnimGraphNode_PoseByName
    poseAsset=/BRCosmetics/Accessories/FORT_Backpacks/Backpack_F_MED_Binary_Twin/Animation/Binary_Twin_Movement_Poses_PoseAsset_Add
    poseName=Run
   }
   n2 = Binary_Twin_Movement_Poses_PoseAsset_Add : Sprint!(PoseWeight:float=1) -> (Pose:Pose Link Structure) {
    is=title AnimGraphNode_PoseByName
    poseAsset=&Binary_Twin_Movement_Poses_PoseAsset_Add
    poseName=Sprint
   }
   n3 = Get SprintSpeedRemap() -> (SprintSpeedRemap:float)
   n4 = Two Way Blend(
    A=Apply Additive(Base=Use cached pose 'preSprintCache'(), Additive=@n2, Alpha=@n3) {alphaBoolBlend {blendOption=Linear}, alphaScaleBiasClamp {bInterpResult=true, interpSpeedDecreasing=5}}
    B=Apply Additive(Base=Use cached pose 'preSprintCache'(), Additive=@n1, Alpha=Get RunSpeedRemap() -> (RunSpeedRemap:float)) {alphaBoolBlend {blendOption=Linear}, alphaScaleBiasClamp {bInterpResult=true, interpSpeedDecreasing=5, scale=0.9}}
    Alpha=@n3
   ) {blendNode {alphaBoolBlend {blendOption=Linear}, alphaScaleBiasClamp {bInterpResult=true, bias=1, interpSpeedDecreasing=5, scale=-1}}}
   n5 = Use cached pose 'preSprintCache'()
   n6 = Get bIsCrouching()
   Output Pose(Result=Blend Poses by bool(
     BlendPose_0=Use cached pose 'preSprintCache'()
     BlendPose_1=Blend Poses by bool(BlendPose_0=@n5, BlendPose_1=@n4, BlendTime_0=0.3, BlendTime_1=0.3, bActiveValue=@n6) {blendType=Linear}
     BlendTime_0=0.3
     BlendTime_1=0.3
     bActiveValue=Get bIsSkyDiving()
    ) {blendType=Linear}) {layerGroup=None}
   AnimGraphNode_StateMachine_0 = wingPosition() {is=title AnimGraphNode_StateMachine}
   Save cached pose 'preSprintCache'(Pose=@AnimGraphNode_StateMachine_0) {cacheName=preSprintCache}
  ]}
 wingPosition {
  entry=Base
  path=AnimGraph.AnimGraphNode_StateMachine_0.wingPosition
  ids=id
  nodes [
   AnimStateNode_0 = Base() {is=title AnimStateNode}
   AnimStateNode_1 = Targeting() {is=title AnimStateNode}
   AnimStateNode_2 = Jumping() {is=title AnimStateNode}
   AnimStateNode_3 = Falling() {is=title AnimStateNode}
   AnimStateNode_4 = SkyDiveMode() {is=title AnimStateNode}
   AnimStateNode_5 = Crouching() {is=title AnimStateNode}
   AnimStateTransitionNode_1 = AnimStateTransitionNode()
   AnimStateTransitionNode_10 = AnimStateTransitionNode()
   AnimStateTransitionNode_11 = AnimStateTransitionNode()
   AnimStateTransitionNode_12 = AnimStateTransitionNode()
   AnimStateTransitionNode_2 = AnimStateTransitionNode()
   AnimStateTransitionNode_3 = AnimStateTransitionNode()
   AnimStateTransitionNode_4 = AnimStateTransitionNode()
   AnimStateTransitionNode_5 = AnimStateTransitionNode()
   AnimStateTransitionNode_6 = AnimStateTransitionNode()
   AnimStateTransitionNode_7 = AnimStateTransitionNode()
   AnimStateTransitionNode_8 = AnimStateTransitionNode()
   AnimStateTransitionNode_9 = AnimStateTransitionNode()
  ]
  links [
   {
    dst=Crouching
    rule=AnimGraph.AnimGraphNode_StateMachine_0.wingPosition.AnimStateTransitionNode_7.Transition
    src=Base
    via=AnimStateTransitionNode_7
   }
   {
    dst=Falling
    rule=AnimGraph.AnimGraphNode_StateMachine_0.wingPosition.AnimStateTransitionNode_12.Transition
    src=Base
    via=AnimStateTransitionNode_12
   }
   {
    dst=Jumping
    rule=AnimGraph.AnimGraphNode_StateMachine_0.wingPosition.AnimStateTransitionNode_3.Transition
    src=Base
    via=AnimStateTransitionNode_3
   }
   {
    dst=SkyDiveMode
    rule=AnimGraph.AnimGraphNode_StateMachine_0.wingPosition.AnimStateTransitionNode_6.Transition
    src=Base
    via=AnimStateTransitionNode_6
   }
   {
    dst=Targeting
    rule=AnimGraph.AnimGraphNode_StateMachine_0.wingPosition.AnimStateTransitionNode_1.Transition
    src=Base
    via=AnimStateTransitionNode_1
   }
   {
    dst=Base
    rule=AnimGraph.AnimGraphNode_StateMachine_0.wingPosition.AnimStateTransitionNode_8.Transition
    src=Crouching
    via=AnimStateTransitionNode_8
   }
   {
    dst=Base
    rule=AnimGraph.AnimGraphNode_StateMachine_0.wingPosition.AnimStateTransitionNode_11.Transition
    src=Falling
    via=AnimStateTransitionNode_11
   }
   {
    dst=Base
    rule=AnimGraph.AnimGraphNode_StateMachine_0.wingPosition.AnimStateTransitionNode_4.Transition
    src=Jumping
    via=AnimStateTransitionNode_4
   }
   {
    dst=Falling
    rule=AnimGraph.AnimGraphNode_StateMachine_0.wingPosition.AnimStateTransitionNode_5.Transition
    src=Jumping
    via=AnimStateTransitionNode_5
   }
   {
    dst=SkyDiveMode
    rule=AnimGraph.AnimGraphNode_StateMachine_0.wingPosition.AnimStateTransitionNode_10.Transition
    src=Jumping
    via=AnimStateTransitionNode_10
   }
   {
    dst=Base
    rule=AnimGraph.AnimGraphNode_StateMachine_0.wingPosition.AnimStateTransitionNode_9.Transition
    src=SkyDiveMode
    via=AnimStateTransitionNode_9
   }
   {
    dst=Base
    rule=AnimGraph.AnimGraphNode_StateMachine_0.wingPosition.AnimStateTransitionNode_2.Transition
    src=Targeting
    via=AnimStateTransitionNode_2
   }
  ]
 }
 Base {
  path=AnimGraph.AnimGraphNode_StateMachine_0.wingPosition.AnimStateNode_0.Base
  nodes [
   n1 = Get PlayerForwardAcceleration() -> (PlayerForwardAcceleration:float)
   n2 = Get PlayerSidewaysAcceleration() -> (PlayerSidewaysAcceleration:float)
   n3 = Binary_Twin_Movement_Poses_PoseAsset : Base!(PoseWeight:float=1) -> (Pose:Pose Link Structure) {
    is=title AnimGraphNode_PoseByName
    poseAsset=/BRCosmetics/Accessories/FORT_Backpacks/Backpack_F_MED_Binary_Twin/Animation/Binary_Twin_Movement_Poses_PoseAsset
    poseName=Base
   }
   Output Animation Pose(Result=Apply Additive(
     Base=@n3
     Additive=Blendspace Player 'BS_Binary_Twin_Leans_Ground'(X=@n2, Y=@n1) {blendSpace=/BRCosmetics/Accessories/FORT_Backpacks/Backpack_F_MED_Binary_Twin/Animation/BS_Binary_Twin_Leans_Ground}
    ) {alphaBoolBlend {blendOption=Linear}})
  ]
 }
 Targeting {
  path=AnimGraph.AnimGraphNode_StateMachine_0.wingPosition.AnimStateNode_1.Targeting
  nodes [
   n1 = Binary_Twin_Movement_Poses_PoseAsset : Targeting!(PoseWeight:float=1) -> (Pose:Pose Link Structure) {
    is=title AnimGraphNode_PoseByName
    poseAsset=&Binary_Twin_Movement_Poses_PoseAsset
    poseName=Targeting
   }
   Output Animation Pose(Result=@n1)
  ]
 }
 Jumping {
  path=AnimGraph.AnimGraphNode_StateMachine_0.wingPosition.AnimStateNode_2.Jumping
  nodes [
   n1 = Get PlayerForwardAcceleration()
   n2 = Get PlayerSidewaysAcceleration()
   n3 = Binary_Twin_Movement_Poses_PoseAsset : Jump!(PoseWeight:float=1) -> (Pose:Pose Link Structure) {
    is=title AnimGraphNode_PoseByName
    poseAsset=&Binary_Twin_Movement_Poses_PoseAsset
    poseName=Jump
   }
   Output Animation Pose(Result=Apply Additive(
     Base=@n3
     Additive=Blendspace Player 'BS_Binary_Twin_Leans_Ground'(X=@n2, Y=@n1) {blendSpace=&BS_Binary_Twin_Leans_Ground}
    ) {alphaBoolBlend {blendOption=Linear}})
  ]
 }
 Falling {
  path=AnimGraph.AnimGraphNode_StateMachine_0.wingPosition.AnimStateNode_3.Falling
  nodes [
   n1 = Get PlayerForwardAcceleration()
   n2 = Get PlayerSidewaysAcceleration()
   n3 = Binary_Twin_Movement_Poses_PoseAsset : Jump(PoseWeight=1) {poseAsset=&Binary_Twin_Movement_Poses_PoseAsset, poseName=Jump}
   Output Animation Pose(Result=Apply Additive(
     Base=@n3
     Additive=Blendspace Player 'BS_Binary_Twin_Leans_Ground'(X=@n2, Y=@n1) {blendSpace=&BS_Binary_Twin_Leans_Ground}
    ) {alphaBoolBlend {blendOption=Linear}})
  ]
 }
 SkyDiveMode {
  path=AnimGraph.AnimGraphNode_StateMachine_0.wingPosition.AnimStateNode_4.SkyDiveMode
  nodes [
   AnimGraphNode_StateMachine_0 = Skydiving State() {is=title AnimGraphNode_StateMachine}
   Output Animation Pose(Result=@AnimGraphNode_StateMachine_0)
  ]
 }
 Skydiving State {
  entry=Parachuting
  path=AnimGraph.AnimGraphNode_StateMachine_0.wingPosition.AnimStateNode_4.SkyDiveMode.AnimGraphNode_StateMachine_0.Skydiving State
  ids=id
  nodes [
   AnimStateNode_0 = Parachuting() {is=title AnimStateNode}
   AnimStateNode_1 = Diving() {is=title AnimStateNode}
   AnimStateNode_2 = Floating() {is=title AnimStateNode}
   AnimStateTransitionNode_13 = AnimStateTransitionNode()
   AnimStateTransitionNode_14 = AnimStateTransitionNode()
   AnimStateTransitionNode_15 = AnimStateTransitionNode()
   AnimStateTransitionNode_16 = AnimStateTransitionNode()
   AnimStateTransitionNode_17 = AnimStateTransitionNode()
   AnimStateTransitionNode_18 = AnimStateTransitionNode()
  ]
  links [
   {
    dst=Floating
    rule=AnimGraph.AnimGraphNode_StateMachine_0.wingPosition.AnimStateNode_4.SkyDiveMode.AnimGraphNode_StateMachine_0.Skydiving State.AnimStateTransitionNode_16.Transition
    src=Diving
    via=AnimStateTransitionNode_16
   }
   {
    dst=Parachuting
    rule=AnimGraph.AnimGraphNode_StateMachine_0.wingPosition.AnimStateNode_4.SkyDiveMode.AnimGraphNode_StateMachine_0.Skydiving State.AnimStateTransitionNode_18.Transition
    src=Diving
    via=AnimStateTransitionNode_18
   }
   {
    dst=Diving
    rule=AnimGraph.AnimGraphNode_StateMachine_0.wingPosition.AnimStateNode_4.SkyDiveMode.AnimGraphNode_StateMachine_0.Skydiving State.AnimStateTransitionNode_15.Transition
    src=Floating
    via=AnimStateTransitionNode_15
   }
   {
    dst=Parachuting
    rule=AnimGraph.AnimGraphNode_StateMachine_0.wingPosition.AnimStateNode_4.SkyDiveMode.AnimGraphNode_StateMachine_0.Skydiving State.AnimStateTransitionNode_17.Transition
    src=Floating
    via=AnimStateTransitionNode_17
   }
   {
    dst=Diving
    rule=AnimGraph.AnimGraphNode_StateMachine_0.wingPosition.AnimStateNode_4.SkyDiveMode.AnimGraphNode_StateMachine_0.Skydiving State.AnimStateTransitionNode_13.Transition
    src=Parachuting
    via=AnimStateTransitionNode_13
   }
   {
    dst=Floating
    rule=AnimGraph.AnimGraphNode_StateMachine_0.wingPosition.AnimStateNode_4.SkyDiveMode.AnimGraphNode_StateMachine_0.Skydiving State.AnimStateTransitionNode_14.Transition
    src=Parachuting
    via=AnimStateTransitionNode_14
   }
  ]
 }
 Parachuting {
  path=AnimGraph.AnimGraphNode_StateMachine_0.wingPosition.AnimStateNode_4.SkyDiveMode.AnimGraphNode_StateMachine_0.Skydiving State.AnimStateNode_0.Parachuting
  nodes [
   n1 = Get PlayerForwardAcceleration()
   n2 = Get PlayerSidewaysAcceleration()
   n3 = Binary_Twin_Movement_Poses_PoseAsset : Float!(PoseWeight:float=1) -> (Pose:Pose Link Structure) {
    is=title AnimGraphNode_PoseByName
    poseAsset=&Binary_Twin_Movement_Poses_PoseAsset
    poseName=Float
   }
   Output Animation Pose(Result=Apply Additive(
     Base=@n3
     Additive=Blendspace Player 'BS_Binary_Twin_Leans_Parachuting'(X=@n2, Y=@n1) {blendSpace=/BRCosmetics/Accessories/FORT_Backpacks/Backpack_F_MED_Binary_Twin/Animation/BS_Binary_Twin_Leans_Parachuting}
    ) {alphaBoolBlend {blendOption=Linear}})
  ]
 }
 Diving {
  path=AnimGraph.AnimGraphNode_StateMachine_0.wingPosition.AnimStateNode_4.SkyDiveMode.AnimGraphNode_StateMachine_0.Skydiving State.AnimStateNode_1.Diving
  nodes [
   n1 = Get PlayerForwardAcceleration()
   n2 = Get PlayerSidewaysAcceleration()
   n3 = Binary_Twin_Movement_Poses_PoseAsset : Dive!(PoseWeight:float=1) -> (Pose:Pose Link Structure) {
    is=title AnimGraphNode_PoseByName
    poseAsset=&Binary_Twin_Movement_Poses_PoseAsset
    poseName=Dive
   }
   Output Animation Pose(Result=Apply Additive(
     Base=@n3
     Additive=Blendspace Player 'BS_Binary_Twin_Leans_Diving'(X=@n2, Y=@n1) {blendSpace=/BRCosmetics/Accessories/FORT_Backpacks/Backpack_F_MED_Binary_Twin/Animation/BS_Binary_Twin_Leans_Diving}
    ) {alphaBoolBlend {blendOption=Linear}})
  ]
 }
 Floating {
  path=AnimGraph.AnimGraphNode_StateMachine_0.wingPosition.AnimStateNode_4.SkyDiveMode.AnimGraphNode_StateMachine_0.Skydiving State.AnimStateNode_2.Floating
  nodes [
   n1 = Get PlayerForwardAcceleration()
   n2 = Get PlayerSidewaysAcceleration()
   n3 = Binary_Twin_Movement_Poses_PoseAsset : Float(PoseWeight=1) {poseAsset=&Binary_Twin_Movement_Poses_PoseAsset, poseName=Float}
   Output Animation Pose(Result=Apply Additive(
     Base=@n3
     Additive=Blendspace Player 'BS_Binary_Twin_Leans_Floating'(X=@n2, Y=@n1) {blendSpace=/BRCosmetics/Accessories/FORT_Backpacks/Backpack_F_MED_Binary_Twin/Animation/BS_Binary_Twin_Leans_Floating}
    ) {alphaBoolBlend {blendOption=Linear}})
  ]
 }
 Transition {
  path=AnimGraph.AnimGraphNode_StateMachine_0.wingPosition.AnimStateNode_4.SkyDiveMode.AnimGraphNode_StateMachine_0.Skydiving State.AnimStateTransitionNode_13.Transition
  nodes [Result(bCanEnterTransition=NOT Boolean(A=Get bIsParachuteOpened()))]
 }
 Transition {
  path=AnimGraph.AnimGraphNode_StateMachine_0.wingPosition.AnimStateNode_4.SkyDiveMode.AnimGraphNode_StateMachine_0.Skydiving State.AnimStateTransitionNode_14.Transition
  nodes [Result(bCanEnterTransition=NOT Boolean(A=Get bIsParachuteOpened()))]
 }
 Transition {
  path=AnimGraph.AnimGraphNode_StateMachine_0.wingPosition.AnimStateNode_4.SkyDiveMode.AnimGraphNode_StateMachine_0.Skydiving State.AnimStateTransitionNode_15.Transition
  nodes [Result(bCanEnterTransition=Get bIsSkydiveDiveMode())]
 }
 Transition {
  path=AnimGraph.AnimGraphNode_StateMachine_0.wingPosition.AnimStateNode_4.SkyDiveMode.AnimGraphNode_StateMachine_0.Skydiving State.AnimStateTransitionNode_16.Transition
  nodes [Result(bCanEnterTransition=NOT Boolean(A=Get bIsSkydiveDiveMode()))]
 }
 Transition {
  path=AnimGraph.AnimGraphNode_StateMachine_0.wingPosition.AnimStateNode_4.SkyDiveMode.AnimGraphNode_StateMachine_0.Skydiving State.AnimStateTransitionNode_17.Transition
  nodes [Result(bCanEnterTransition=Get bIsParachuteOpened())]
 }
 Transition {
  path=AnimGraph.AnimGraphNode_StateMachine_0.wingPosition.AnimStateNode_4.SkyDiveMode.AnimGraphNode_StateMachine_0.Skydiving State.AnimStateTransitionNode_18.Transition
  nodes [Result(bCanEnterTransition=Get bIsParachuteOpened())]
 }
 Crouching {
  path=AnimGraph.AnimGraphNode_StateMachine_0.wingPosition.AnimStateNode_5.Crouching
  nodes [
   n1 = Binary_Twin_Movement_Poses_PoseAsset : Crouch!(PoseWeight:float=1) -> (Pose:Pose Link Structure) {
    is=title AnimGraphNode_PoseByName
    poseAsset=&Binary_Twin_Movement_Poses_PoseAsset
    poseName=Crouch
   }
   Output Animation Pose(Result=@n1)
  ]
 }
 Transition {
  path=AnimGraph.AnimGraphNode_StateMachine_0.wingPosition.AnimStateTransitionNode_1.Transition
  nodes [Result(bCanEnterTransition=Get bIsTargeting())]
 }
 Transition {
  path=AnimGraph.AnimGraphNode_StateMachine_0.wingPosition.AnimStateTransitionNode_10.Transition
  nodes [Result(bCanEnterTransition=Get bIsSkyDiving())]
 }
 Transition {
  path=AnimGraph.AnimGraphNode_StateMachine_0.wingPosition.AnimStateTransitionNode_11.Transition
  nodes [Result(bCanEnterTransition=NOT Boolean(A=Get bIsFalling()))]
 }
 Transition {
  path=AnimGraph.AnimGraphNode_StateMachine_0.wingPosition.AnimStateTransitionNode_12.Transition
  nodes [Result(bCanEnterTransition=Get bIsFalling())]
 }
 Transition {
  path=AnimGraph.AnimGraphNode_StateMachine_0.wingPosition.AnimStateTransitionNode_2.Transition
  nodes [Result(bCanEnterTransition=NOT Boolean(A=Get bIsTargeting()))]
 }
 Transition {
  path=AnimGraph.AnimGraphNode_StateMachine_0.wingPosition.AnimStateTransitionNode_3.Transition
  nodes [Result(bCanEnterTransition=Get bIsJumping())]
 }
 Transition {
  path=AnimGraph.AnimGraphNode_StateMachine_0.wingPosition.AnimStateTransitionNode_4.Transition
  nodes [Result(bCanEnterTransition=NOT Boolean(A=Get bIsJumping()))]
 }
 Transition {
  path=AnimGraph.AnimGraphNode_StateMachine_0.wingPosition.AnimStateTransitionNode_5.Transition
  nodes [Result(bCanEnterTransition=Get bIsFalling())]
 }
 Transition {
  path=AnimGraph.AnimGraphNode_StateMachine_0.wingPosition.AnimStateTransitionNode_6.Transition
  nodes [Result(bCanEnterTransition=Get bIsSkyDiving())]
 }
 Transition {
  path=AnimGraph.AnimGraphNode_StateMachine_0.wingPosition.AnimStateTransitionNode_7.Transition
  nodes [Result(bCanEnterTransition=Get bIsCrouching())]
 }
 Transition {
  path=AnimGraph.AnimGraphNode_StateMachine_0.wingPosition.AnimStateTransitionNode_8.Transition
  nodes [Result(bCanEnterTransition=NOT Boolean(A=Get bIsCrouching()))]
 }
 Transition {
  path=AnimGraph.AnimGraphNode_StateMachine_0.wingPosition.AnimStateTransitionNode_9.Transition
  nodes [Result(bCanEnterTransition=NOT Boolean(A=Get bIsSkyDiving()))]
 }
 EventGraph {nodes [
   TryGetPawnOwner()
   Event BlueprintUpdateAnimation()
  ]}
]
variables []
